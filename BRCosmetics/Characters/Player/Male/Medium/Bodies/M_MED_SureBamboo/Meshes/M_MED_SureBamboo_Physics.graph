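# Physics asset for the male medium M_MED_SureBamboo player body, covering only the hips and belt accessories.
# Capsules wrap the pelvis, both thighs, the left belt pouch and the right holster, with each constrained to the pelvis.
# Pouch and holster get narrow swing limits with zero-degree twist limits, suggesting secondary motion against the legs.
# The thigh constraints are asymmetric: the left has its angular motions explicitly free, unlike the right.
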PhysicsAsset /BRCosmetics/Characters/Player/Male/Medium/Bodies/M_MED_SureBamboo/Meshes/M_MED_SureBamboo_Physics
bodies {
 deform_belt_pouch_l [{
   shapeType=Capsule
   length=3.7975
   radius=5.5375
   shapeName=deform_belt_pouch_l_capsule
   center=<11.9464 -0.0287 5.872>
   rotation=<rot 0 0.0001 95>
  }]
 deform_holster_r [{
   shapeType=Capsule
   length=17.3735
   radius=5.1125
   shapeName=deform_holster_r_capsule
   center=<13.4479 2.4643 6.5596>
   rotation=<rot 90 0 0.0021>
  }]
 pelvis [{shapeType=Capsule, length=7.4287, radius=12.1714, shapeName=pelvis_capsule, center=<-0.7508 -1.222 0.0137>}]
 thigh_l [{shapeType=Capsule, length=25.7144, radius=8.2463, shapeName=thigh_l_capsule, center=<-12.8029 -0.0444 -0.5665>, rotation=<rot 90 0 0>}]
 thigh_r [{shapeType=Capsule, length=25.7144, radius=8.2463, shapeName=thigh_l_capsule, center=<12.8029 -0.0444 -0.5665>, rotation=<rot 90 0 0>}]
}
constraints [
 {bone1Name=thigh_r, bone2Name=pelvis, swing1LimitDegrees=45, swing2LimitDegrees=45, twistLimitDegrees=45}
 {
  bone1Name=thigh_l
  bone2Name=pelvis
  swing1LimitDegrees=45
  swing1Motion=Free
  swing2LimitDegrees=45
  swing2Motion=Free
  twistLimitDegrees=45
  twistMotion=Free
 }
 {bone1Name=deform_belt_pouch_l, bone2Name=pelvis, swing1LimitDegrees=20, swing2LimitDegrees=20, twistLimitDegrees=0}
 {bone1Name=deform_holster_r, bone2Name=pelvis, swing1LimitDegrees=15, swing2LimitDegrees=15, twistLimitDegrees=0}
]
