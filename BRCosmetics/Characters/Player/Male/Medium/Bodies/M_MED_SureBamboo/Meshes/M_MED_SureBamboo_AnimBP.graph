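# Animation Blueprint for the SureBamboo medium male player body cosmetic, adding procedural and secondary motion to a pose copied from the attached parent mesh.
# The copied pose runs through the body's Control Rig, driven by gameplay-state flags (targeting, crouch, DBNO, parachute, emote and others) that the EventGraph smooths into floats with FInterpTo.
# The cached result feeds RigidBody simulation when bRigidBodyEnabled is set, otherwise cheaper AnimDynamics on the right holster and left belt pouch.
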AnimBlueprint /BRCosmetics/Characters/Player/Male/Medium/Bodies/M_MED_SureBamboo/Meshes/M_MED_SureBamboo_AnimBP
target_skeleton=&/M_MED_SureBamboo_Skeleton
graphs [
 AnimGraph {nodes [
   n1 = RigidBody(ComponentPose=Local To Component(LocalPose=Use cached pose 'Post Control Rig'())) {
    bOverrideWorldGravity=true
    overridePhysicsAsset=&/M_MED_SureBamboo_Physics
    simulationSpace=BaseBoneSpace
    alphaBoolBlend {blendOption=Linear}
    baseBoneRef {boneName=pelvis}
    componentAppliedLinearAccClamp=<2000 2000 2000>
    componentLinearAccScale=<0.3 0.3 0.3>
    componentLinearVelScale=<0.3 0.3 0.3>
    overrideWorldGravity {z=-980}
   }
   n2 = Anim Dynamics - Bone: deform_holster_r_MSTR(ComponentPose=Local To Component(LocalPose=Use cached pose 'Post Control Rig'()), Alpha=1, ExternalForce="0,0,0") {
    angularDampingOverride=1
    bOverrideAngularDamping=true
    bUseGravityOverride=true
    simulationSpace=BoneRelative
    alphaBoolBlend {blendOption=Linear}
    boundBone {boneName=deform_holster_r_MSTR}
    componentAppliedLinearAccClamp=<2000 2000 2000>
    componentLinearAccScale=<0.15 0.15 0.15>
    componentLinearVelScale=<0.15 0.15 0.15>
    gravityOverride {z=-500}
    physicsBodyDefinitions [{
      boundBone {boneName=deform_holster_r_MSTR}
      boxExtents=<10 10 10>
      constraintSetup {angularLimitsMax=<0 -1 20>, angularLimitsMin=<0 -20 -20>, angularTarget=<0 0 0>}
      localJointOffset=<5 0 0>
     }]
    relativeSpaceBone {boneName=root}
   }
   n3 = Anim Dynamics - Bone: deform_belt_pouch_l_MSTR(ComponentPose=@n2, Alpha=1, ExternalForce="0,0,0") {
    angularDampingOverride=1
    bOverrideAngularDamping=true
    bUseGravityOverride=true
    simulationSpace=BoneRelative
    alphaBoolBlend {blendOption=Linear}
    boundBone {boneName=deform_belt_pouch_l_MSTR}
    componentAppliedLinearAccClamp=<2000 2000 2000>
    componentLinearAccScale=<0.1 0.1 0.1>
    componentLinearVelScale=<0.1 0.1 0.1>
    gravityOverride {z=-500}
    physicsBodyDefinitions [{
      boundBone {boneName=deform_belt_pouch_l_MSTR}
      boxExtents=<15 15 15>
      constraintSetup {angularLimitsMax=<0 -1 20>, angularLimitsMin=<0 -20 -20>, angularTarget=<0 0 0>}
      localJointOffset=<5 0 0>
     }]
    relativeSpaceBone {boneName=root}
   }
   n4 = Blend Poses by bool(BlendPose_0=Component To Local(ComponentPose=@n1), BlendPose_1=Component To Local(ComponentPose=@n3), bActiveValue=Get bRigidBodyEnabled()) {blendType=Linear}
   Output Pose(Result=Component To Local(ComponentPose=Local To Component(LocalPose=@n4))) {layerGroup=None}
   n5 = Control Rig!(
    Source=Copy Pose From Mesh() {bUseAttachedParent=true}
    Alpha
    IsTargeting:double=Get Is Targeting() -> (Is Targeting:double)
    IsCrouching:double=Get IsCrouching() -> (IsCrouching:double)
    IsMoving2D:double=Get Is Moving 2D() -> (Is Moving 2D:double)
    IsNOTCrouchingOrDBNO:double=Get Is NOTCrouching or DBNO() -> (Is NOTCrouching or DBNO:double)
    IsNOTPlayingMeleeAnim:double=Get Is NOTPlaying Melee Anim() -> (Is NOTPlaying Melee Anim:double)
    IsDBNO:double=Get Is DBNO() -> (Is DBNO:double)
    IsParachuteOpened:double=Get Is Parachute Opened() -> (Is Parachute Opened:double)
    IsFrontEnd:double=Get Is Front End() -> (Is Front End:double)
    IsPlayingEmote:double=Get Is Playing Emote() -> (Is Playing Emote:double)
   ) -> (Pose) {
    alphaBoolBlend {blendOption=Linear}
    controlRigAssetReference {blueprintRigClass=&/M_MED_SureBamboo_CtrlRig.M_MED_SureBamboo_CtrlRig_C}
   }
   Save cached pose 'Post Control Rig'(Pose=@n5) {cacheName=Post Control Rig}
  ]}
 EventGraph {nodes [
   n1 = Event BlueprintUpdateAnimation()
   > Set Delta Time X(Delta Time X=@n1.DeltaTimeX)
   n2 = Get bIsTargeting()
   > Set Is Targeting(Is Targeting:double=FInterpTo(Current=Get Is Targeting(), Target="To Float (Boolean)"(InBool=@n2), DeltaTime=Get Delta Time X(), InterpSpeed=10)) -> (Output_Get:double)
   n3 = Get bIsCrouching()
   > Set IsCrouching(IsCrouching:double=FInterpTo(Current=Get IsCrouching(), Target="To Float (Boolean)"(InBool=@n3), DeltaTime=Get Delta Time X(), InterpSpeed=10)) -> (Output_Get:double)
   n4 = Get bIsMoving2D()
   > Set Is Moving 2D(Is Moving 2D:double=FInterpTo(Current=Get Is Moving 2D(), Target="To Float (Boolean)"(InBool=@n4), DeltaTime=Get Delta Time X(), InterpSpeed=10)) -> (Output_Get:double)
   n5 = NOR Boolean!(A:bool=Get bIsDBNO(), B:bool=Get bIsCrouching()) -> (ReturnValue:bool)
   > Set Is NOTCrouching or DBNO(Is NOTCrouching or DBNO:double=FInterpTo(Current=Get Is NOTCrouching or DBNO(), Target="To Float (Boolean)"(InBool=@n5), DeltaTime=Get Delta Time X(), InterpSpeed=10)) -> (Output_Get:double)
   n6 = NOT Boolean(A=Get bIsPlayingMeleeAnim())
   > Set Is NOTPlaying Melee Anim(Is NOTPlaying Melee Anim:double=FInterpTo(Current=Get Is NOTPlaying Melee Anim(), Target="To Float (Boolean)"(InBool=@n6), DeltaTime=Get Delta Time X(), InterpSpeed=1)) -> (Output_Get:double)
   n7 = Get bIsDBNO()
   > Set Is DBNO(Is DBNO:double=FInterpTo(Current=Get Is DBNO(), Target="To Float (Boolean)"(InBool=@n7), DeltaTime=Get Delta Time X(), InterpSpeed=10)) -> (Output_Get:double)
   n8 = Get bIsParachuteOpened()
   > Set Is Parachute Opened(Is Parachute Opened:double=FInterpTo(Current=Get Is Parachute Opened(), Target="To Float (Boolean)"(InBool=@n8), DeltaTime=Get Delta Time X(), InterpSpeed=10)) -> (Output_Get:double)
   n9 = Get bIsInFrontEnd()
   > Set Is Front End(Is Front End:double=FInterpTo(Current=Get Is Front End(), Target="To Float (Boolean)"(InBool=@n9), DeltaTime=Get Delta Time X(), InterpSpeed=10)) -> (Output_Get:double)
   n10 = Get bIsPlayingEmote()
   > Set Is Playing Emote(Is Playing Emote:double=FInterpTo(Current=Get Is Playing Emote(), Target="To Float (Boolean)"(InBool=@n10), DeltaTime=Get Delta Time X(), InterpSpeed=50)) -> (Output_Get:double)
  ]}
]
variables [
 Delta Time X {type=number}
 Is DBNO {type=number}
 Is Front End {type=number}
 Is Moving 2D {type=number}
 Is NOTCrouching or DBNO {type=number}
 Is NOTPlaying Melee Anim {type=number}
 Is Parachute Opened {type=number}
 Is Playing Emote {type=number}
 Is Targeting {type=number}
 IsCrouching {type=number}
]
